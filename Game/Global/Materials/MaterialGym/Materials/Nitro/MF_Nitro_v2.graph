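# Material function that renders an animated, glowing nitro-liquid look with bubbles and glass edges over incoming material attributes, selectable by static switches and a SkinFX mask.
# Custom HLSL ray-marches distorted swirl noise for depth, with platform-scaled steps and a single-step node fallback for low-end targets.
# That height, a density mask and Volatile pulsing drive emissive color.
# Two bump-offset bubble layers, Fresnel edging and an optional occlusion mask finish color and normal.
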MaterialFunction /Game/Global/Materials/MaterialGym/Materials/Nitro/MF_Nitro_v2
graph {nodes [
  n1 = StaticSwitchParameter(True=NamedRerouteUsage(), False=BlendMaterialAttributes(A=NamedRerouteUsage(), B=NamedRerouteUsage(), Alpha=NamedRerouteUsage())) {defaultValue=true, desc=when disabled you get the texture set inputs, group=04_Tech Art, parameterName=UseOnlyNitro, sortPriority=-4}
  FunctionOutput(None=StaticSwitchParameter(True=@n1, False=NamedRerouteUsage()) {defaultValue=true, group=04_Tech Art, parameterName=UseNitro, sortPriority=-5})
  MaterialFunctionCall!(Vector3=PreSkinnedLocalBounds().Max, MaximumNumberOfDigits, UVs, DebugTextLocation RG_UpperRight BA_LowerLeft, Component Spacing) {materialFunction=/Engine/Functions/Engine_MaterialFunctions02/Utility/DebugFloat3Values}
  MaterialFunctionCall() {materialFunction=&MF_Hash}
  NamedRerouteDeclaration(None=StaticBoolParameter() {defaultValue=true, group=Nitro, parameterName=UseBubbles, sortPriority=89}) {name=UseBubbles, nodeColor=<rgba 0.125 1 0 1>}
  n2 = NamedRerouteUsage()
  n3 = ScalarParameter() {defaultValue=1, group=Nitro, parameterName=UseFlashing, sortPriority=250}
  n4 = ComponentMask(None=NamedRerouteUsage()) {r=true}
  n5 = LinearInterpolate(A=Power(Base=@n4), B=@n4, Alpha=NamedRerouteUsage())
  n6 = NamedRerouteUsage()
  n7 = LinearInterpolate(A=Power(Base=Multiply(A=@n6) {constB=2}), Alpha=NamedRerouteUsage())
  NamedRerouteDeclaration(None=Multiply(A=LinearInterpolate(B=@n2, Alpha=@n3) {constA=1}, B=Multiply(A=@n5, B=@n7))) {name=Lightning/VolatileMask, nodeColor=<rgba 1 0.1172 0 1>}
  n8 = ScalarParameter() {defaultValue=0.1, group=Nitro, parameterName=GlassEdgeMin, sortPriority=401}
  n9 = ScalarParameter() {defaultValue=0.4, group=Nitro, parameterName=GlassEdgeBias, sortPriority=402}
  n10 = VertexNormalWS()
  NamedRerouteDeclaration(None=SmoothStep(Min=@n8, Max=Add(A=@n8, B=@n9), Value=Fresnel(Normal=@n10) {baseReflectFraction=0, exponent=1}) {constMax=0.9, constMin=0.8}) {name=GlassEdgeMask, nodeColor=<rgba 0 1 1 1>}
  n11 = TextureSample(UVs=NamedRerouteUsage()) {
   samplerSource=SSM_Wrap_WorldGroupSettings
   samplerType=SAMPLERTYPE_Normal
   texture=&/T_NitroBubbles_N
  }
  n12 = TextureSample(UVs=NamedRerouteUsage()) {
   samplerSource=SSM_Wrap_WorldGroupSettings
   samplerType=SAMPLERTYPE_Normal
   texture=&T_NitroBubbles_N
  }
  n13 = TextureSample(UVs=NamedRerouteUsage()) {
   samplerSource=SSM_Wrap_WorldGroupSettings
   samplerType=SAMPLERTYPE_LinearColor
   texture=&/T_NitroBubbles
  }
  NamedRerouteDeclaration(None=LinearInterpolate(A=@n11.RGB, B=@n12.RGB, Alpha=@n13.G)) {name=BubbleNormal, nodeColor=<rgba 1 0.4219 0 1>}
  n14 = NamedRerouteUsage()
  n15 = VectorParameter() {group=Nitro, parameterName=BubbleColor, sortPriority=90, defaultValue=<rgba 0.0208 0.0019 0.0006 1>}
  NamedRerouteDeclaration(None=Multiply(A=Multiply(A=@n14, B=@n15.RGB))) {name=BubbleColor, nodeColor=<rgba 0.1016 0 1 1>}
  n16 = TextureSample(UVs=NamedRerouteUsage()) {
   samplerSource=SSM_Wrap_WorldGroupSettings
   samplerType=SAMPLERTYPE_LinearColor
   texture=&T_NitroBubbles
  }
  n17 = SphereMask(A=LinearInterpolate(A=@n16.G, B=@n13.G, Alpha=@n13.G), B=Constant() {r=1}, Radius=NamedRerouteUsage(), Hardness=Constant())
  NamedRerouteDeclaration(None=Saturate(None=Multiply(A=@n17) {constB=5})) {name=BubbleMask, nodeColor=<rgba 0.2188 1 0 1>}
  n18 = VectorParameter() {group=Nitro, parameterName=NitroTile, sortPriority=10, defaultValue=<rgba 1 1 0 1>}
  MaterialExpressionMultiply_52 = Multiply(A=Multiply(A=TextureCoordinate(), B=ComponentMask(None=@n18.RGB) {g=true, r=true}), B=NamedRerouteUsage()) {desc=uv}
  MaterialExpressionTransform_1 = Transform(None=CameraVectorWS()) {transformSourceType=TRANSFORMSOURCE_World, transformType=TRANSFORM_Tangent}
  n19 = Constant() {r=1}
  MaterialExpressionMaterialFunctionCall_0 = MaterialFunctionCall(
   "Xb1,PS4"=Constant() {r=6}
   "PC-High&&Epic,XSX,PS5"=Constant() {r=8}
   "PC-Perf,Android"=@n19
   Switch=@n19
   "PC-Low(NotPerf)&&Medium"=Constant() {r=4}
  ) {materialFunction=&PlatformSwitch_1D}
  MaterialExpressionScalarParameter_20 = ScalarParameter() {defaultValue=0.005, group=Nitro, parameterName=NitroRayMarchStepSize, sliderMax=0.2, sortPriority=13}
  MaterialExpressionTextureObject_1 = TextureObject() {
   samplerType=SAMPLERTYPE_LinearGrayscale
   texture=/Game/Global/Materials/MaterialGym/Materials/SmokeWithFlowmap/Textures/T_Swirly_Noise_MiddleDetail
  }
  MaterialExpressionTextureObject_0 = TextureObject() {
   samplerType=SAMPLERTYPE_LinearGrayscale
   texture=&T_Swirly_Noise_MiddleDetail
  }
  MaterialExpressionTime_4 = Time()
  MaterialExpressionMultiply_49 = Multiply(
   A=ComponentMask(None=VectorParameter() {group=Nitro, parameterName=NitroDistortionSpeed, sortPriority=11, defaultValue=<rgba 0 0.2 0 1>}.RGB) {g=true, r=true}
   B=NamedRerouteUsage()
  )
  MaterialExpressionMultiply_47 = Multiply(
   A=ComponentMask(None=VectorParameter() {group=Nitro, parameterName=NitroHeightSpeed, sortPriority=12, defaultValue=<rgba 0 0.1 0 1>}.RGB) {g=true, r=true}
   B=NamedRerouteUsage()
  )
  n20 = Custom!(
   uv=@MaterialExpressionMultiply_52
   viewDir=@MaterialExpressionTransform_1
   steps=@MaterialExpressionMaterialFunctionCall_0.Result
   stepdist=@MaterialExpressionScalarParameter_20
   tex=@MaterialExpressionTextureObject_1
   dTex=@MaterialExpressionTextureObject_0
   Time=@MaterialExpressionTime_4
   distortionSpeed=@MaterialExpressionMultiply_49
   heightSpeed=@MaterialExpressionMultiply_47
  ) {
   code="float3 rayStep = viewDir * -1;\r\nfloat2 uv2 = uv + Time * distortionSpeed;\r\nuv += Time * heightSpeed;\r\nfloat4 heightTex = Texture2DSample(tex,texSampler,uv);\r\nfloat4 distortionTex = Texture2DSample(dTex,texSampler,uv2);\r\n\r\nint Steps = floor(steps);\r\n//float steps = 16;\r\n//float stepdist = float(0.1);\r\n\r\n//[UNROLL]\r\nfor(int i=0; i <= Steps; i++)\r\n{\r\n\tif( max(heightTex.r, 0.0001) <= i / Steps)\r\n\t{\r\n\t\treturn float((heightTex.r * i) / Steps); \r\n\t\t//return float(i/16.0);\r\n\t}\r\n\r\n\tfloat stepct = heightTex.r * stepdist;\r\n\t//float stepct = 0.05;\r\n\tuv += rayStep * stepct;\r\n\tuv2 += rayStep * stepct;\r\n\tdistortionTex = Texture2DSample(dTex,texSampler,uv2);\r\n\tfloat2 distortStr = distortionTex.xy * 0.1;\r\n\theightTex = Texture2DSample(tex,texSampler,uv.xy + distortStr);\r\n\tRayMarchUV = uv.xy + distortStr;\r\n}\r\n\r\nreturn heightTex;"
   description=RayMarchTexture2
   outputType=CMOT_Float1
   showCode=true
   additionalOutputs [{outputName=RayMarchUV, outputType=CMOT_Float2}]
   inputs [
    {inputName=uv, input {expression=&self:MaterialExpressionMultiply_52}}
    {inputName=viewDir, input {expression=&self:MaterialExpressionTransform_1}}
    {inputName=steps, input {expression=&self:MaterialExpressionMaterialFunctionCall_0}}
    {inputName=stepdist, input {expression=&self:MaterialExpressionScalarParameter_20}}
    {inputName=tex, input {expression=&self:MaterialExpressionTextureObject_1}}
    {inputName=dTex, input {expression=&self:MaterialExpressionTextureObject_0}}
    {inputName=Time, input {expression=&self:MaterialExpressionTime_4}}
    {inputName=distortionSpeed, input {expression=&self:MaterialExpressionMultiply_49}}
    {inputName=heightSpeed, input {expression=&self:MaterialExpressionMultiply_47}}
   ]
  }
  n21 = ComponentMask(None=Add(A=Multiply(A=@MaterialExpressionTime_4, B=@MaterialExpressionMultiply_47), B=@MaterialExpressionMultiply_52) {desc=uv}) {g=true, r=true}
  n22 = TextureSample(UVs=@n21, Tex=@MaterialExpressionTextureObject_1) {desc=heightTex}
  n23 = ComponentMask(None=Multiply(
    A=Multiply(A=@MaterialExpressionTransform_1) {constB=-1, desc=rayStep}
    B=Multiply(A=@n22.R, B=@MaterialExpressionScalarParameter_20) {desc=stepct}
   )) {g=true, r=true}
  n24 = ComponentMask(None=Add(A=Multiply(A=@MaterialExpressionTime_4, B=@MaterialExpressionMultiply_49), B=@MaterialExpressionMultiply_52) {desc=uv2}) {g=true, r=true}
  n25 = TextureSample(UVs=ComponentMask(None=Add(A=@n23, B=@n24) {desc=uv2}) {g=true, r=true}, Tex=@MaterialExpressionTextureObject_1) {desc=distortionTex}
  n26 = Add(A=Add(A=@n23, B=@n21) {desc=uv}, B=Multiply(A=ComponentMask(None=@n25.RGB) {g=true, r=true}) {constB=0.1, desc=distortStr}) {desc=RayMarchUV}
  NamedRerouteDeclaration(None=MaterialFunctionCall!(
    "Xb1,PS4"=@n20.RayMarchUV
    "PC-High&&Epic,XSX,PS5"=@n20.RayMarchUV
    "PC-Perf,Android"=@n26
    Switch=@n26
    "PC-Low(NotPerf)&&Medium"=@n20.RayMarchUV
   ) {materialFunction=/Game/Effects/Niagara/MasterMaterials/Functions/Common/PlatformSwitch_2D}.Result) {name=RayMarchUV, nodeColor=<rgba 1 0 0.5391 1>}
  n27 = NamedRerouteUsage()
  n28 = LinearInterpolate(
   A=ScalarParameter() {defaultValue=1.75, group=Nitro, parameterName=NitroDensityExponentHigh, sortPriority=44}
   B=ScalarParameter() {defaultValue=-0.5, group=Nitro, parameterName=NitroDensityExponentLow, sortPriority=43}
   Alpha=NamedRerouteUsage()
  ) {constB=0}
  NamedRerouteDeclaration(None=Saturate(None=Power(Base=@n27, Exp=@n28) {constExponent=4})) {name=HeightMapModified, nodeColor=<rgba 0.1719 0 1 1>}
  n29 = VectorParameter() {group=Nitro, parameterName=NitroEmisiveColor_02, sortPriority=21, defaultValue=<rgba 0.0208 0.0019 0.0006 1>}
  n30 = VectorParameter() {group=Nitro, parameterName=NitroEmisiveColor_01, sortPriority=20, defaultValue=<rgba 1 0.3866 0 1>}
  n31 = Multiply(
   A=LinearInterpolate(A=@n29.RGB, B=NamedRerouteDeclaration(None=@n30.RGB) {name=NitroEmissiveColor, nodeColor=<rgba 0.1484 1 0 1>}, Alpha=NamedRerouteUsage())
   B=ScalarParameter() {defaultValue=3, group=Nitro, parameterName=NitroBrightness, sortPriority=23}
  ) {constB=3}
  n32 = VectorParameter() {group=Nitro, parameterName=NitroEmisiveColor_03, sortPriority=22, defaultValue=<rgba 0.3177 0.0283 0.0093 1>}
  n33 = Multiply(
   A=LinearInterpolate(A=@n32.RGB, B=NamedRerouteUsage(), Alpha=NamedRerouteUsage())
   B=ScalarParameter() {defaultValue=10, group=Nitro, parameterName=VolatileBrightness, sortPriority=24}
  )
  NamedRerouteDeclaration(None=LinearInterpolate(A=Power(Base=@n31) {constExponent=1.5}, B=Power(Base=@n33) {constExponent=3}, Alpha=NamedRerouteUsage()) {constB=0.5}) {name=NitroColor, nodeColor=<rgba 0.4062 0 1 1>}
  n34 = TextureSample(UVs=NamedRerouteUsage()) {
   samplerSource=SSM_Wrap_WorldGroupSettings
   samplerType=SAMPLERTYPE_Normal
   texture=/Game/Characters/Player/Male/Large/Bodies/M_LRG_Clash_V/Textures/Simple_Perlin_N
  }
  n35 = LinearInterpolate(A=@n34.RGB, B=NamedRerouteUsage(), Alpha=NamedRerouteUsage())
  n36 = NamedRerouteUsage()
  n37 = NamedRerouteUsage()
  n38 = ScalarParameter() {defaultValue=0.7, group=Nitro, parameterName=GlassEdgeNormalStrength, sortPriority=411}
  NamedRerouteDeclaration(None=LinearInterpolate(A=StaticSwitch(True=@n35, False=@n34.RGB, Value=@n36), B=Constant3Vector() {constant=<rgba 0 0 1 0>}, Alpha=Multiply(A=@n37, B=@n38))) {name=OUT_Normal, nodeColor=<rgba 0.9688 1 0 1>}
  n39 = TextureCoordinate()
  n40 = VectorParameter() {group=Nitro, parameterName=BubbleDistortionSpeed, sortPriority=301, defaultValue=<rgba 0 0.25 0 1>}
  n41 = Multiply(A=Time(), B=Multiply(A=ComponentMask(None=@n40.RGB) {g=true, r=true}, B=NamedRerouteUsage()))
  n42 = ComponentMask(None=VectorParameter() {group=Nitro, parameterName=BubbleDistortionTiling, sortPriority=300, defaultValue=<rgba 2 2 1 1>}.RGB) {g=true, r=true}
  n43 = NamedRerouteUsage()
  n44 = TextureSample(UVs=Multiply(A=Add(A=@n39, B=@n41), B=Multiply(A=@n42, B=@n43))) {samplerSource=SSM_Wrap_WorldGroupSettings, texture=/Game/Characters/Player/Male/Medium/Bodies/M_MED_Viper/Textures/Glitch/T_SoftNoise}
  n45 = Subtract(A=Multiply(A=@n44.RGB) {constB=2})
  NamedRerouteDeclaration(None=Multiply(
    A=ComponentMask(None=@n45) {g=true, r=true}
    B=ScalarParameter() {defaultValue=0.01, group=Nitro, parameterName=BubbleDistortionStrength, sortPriority=302}
   )) {name=BubbleDistortion, nodeColor=<rgba 0.6641 1 0 1>}
  n46 = TextureSampleParameter2D(UVs=TextureCoordinate() {coordinateIndex=1}) {
   group=05_Nitro
   parameterName=OccludeMask
   samplerSource=SSM_Wrap_WorldGroupSettings
   texture=&T_BlacksRGB
  }
  NamedRerouteDeclaration(None=StaticSwitchParameter(
    True=ChannelMaskParameter(None=@n46.RGB) {group=05_Nitro, maskChannel=Green, parameterName=NitroOcclude_Mask, defaultValue=<rgba 0 1 0 0>}.R
    False=Constant()
   ) {group=05_Nitro, parameterName=UseOccludeMask}) {name=Occlude_Mask, nodeColor=<rgba 1 0 0.7969 1>}
  n47 = FunctionInput() {inputName=MA, inputType=FunctionInput_MaterialAttributes}
  NamedRerouteDeclaration(None=GetMaterialAttributes(None=NamedRerouteDeclaration(None=@n47) {name=MA_In, nodeColor=<rgba 0 0.6484 1 1>}) {attributeGetTypes [#1]}.BaseColor) {name=BaseColor_In, nodeColor=<rgba 1 0 0.6797 1>}
  n48 = TextureCoordinate()
  n49 = TextureSample(UVs=Panner(Coordinate=Multiply(A=@n48) {constB=2}) {speedY=0.1}) {samplerSource=SSM_Wrap_WorldGroupSettings, texture=&T_SoftNoise}
  n50 = LinearInterpolate(Alpha=@n49.R) {constA=0.5}
  n51 = LocalPosition() {localOrigin=InstancePreSkinning}
  n52 = PreSkinnedLocalBounds()
  n53 = Divide(A=Add(A=VertexInterpolator(VS=@n51.Z).PS, B=ComponentMask(None=@n52.Half Extents) {b=true}), B=ComponentMask(None=@n52.Extents) {b=true}) {constB=250}
  n54 = ScalarParameter() {defaultValue=0.5, group=Nitro, parameterName=BubbleSizeGradient_Scale, sliderMax=2}
  n55 = Saturate(None=Add(
    A=Divide(A=@n53, B=@n54) {constB=250}
    B=ScalarParameter() {group=Nitro, parameterName=BubbleSizeGradient_Offset, sliderMax=16, sliderMin=-16}
   ) {constB=0.5})
  n56 = LinearInterpolate(
   A=@n55
   B=OneMinus(None=@n55)
   Alpha=Round(None=ScalarParameter() {group=Nitro, parameterName=BubbleSizeGradient_InvertGradient, sliderMax=1})
  )
  NamedRerouteDeclaration(None=LinearInterpolate(A=@n50, B=Multiply(A=@n50, B=@n56), Alpha=ScalarParameter() {group=Nitro, parameterName=BubbleSizeMap_Intensity})) {name=BubbleMaskRadius, nodeColor=<rgba 0.3594 0 1 1>}
  n57 = VectorParameter() {group=Nitro, parameterName=BubbleTiling_01, sortPriority=100, defaultValue=<rgba 1 1 0 1>}
  n58 = Multiply(
   A=BumpOffset(
    Coordinate=TextureCoordinate()
    Height=Constant() {r=-1}
    HeightRatioInput=ScalarParameter() {defaultValue=0.3, group=Nitro, parameterName=BubbleBumpOffset_01, sortPriority=102}
   ) {heightRatio=0.1}
   B=Multiply(A=ComponentMask(None=@n57.RGB) {g=true, r=true}, B=NamedRerouteUsage())
  ) {constB=3}
  n59 = Multiply(
   A=NamedRerouteUsage()
   B=ComponentMask(None=VectorParameter() {group=Nitro, parameterName=BubbleSpeed_01, sortPriority=101, defaultValue=<rgba 0 0.25 0.2 1>}.RGB) {g=true, r=true}
  )
  NamedRerouteDeclaration(None=Add(A=Panner(Coordinate=@n58, Speed=@n59) {speedY=-0.4}, B=NamedRerouteUsage())) {name=Bubble_01_UV, nodeColor=<rgba 0 1 0.0391 1>}
  n60 = VectorParameter() {group=Nitro, parameterName=BubbleTiling_02, sortPriority=200, defaultValue=<rgba 1 1 0 1>}
  n61 = Multiply(
   A=BumpOffset(
    Coordinate=TextureCoordinate()
    Height=Constant() {r=-1}
    HeightRatioInput=ScalarParameter() {defaultValue=0.2, group=Nitro, parameterName=BubbleBumpOffset_02, sortPriority=202}
   ) {heightRatio=0.1}
   B=Multiply(A=ComponentMask(None=@n60.RGB) {g=true, r=true}, B=NamedRerouteUsage())
  ) {constB=3}
  n62 = Multiply(
   A=NamedRerouteUsage()
   B=ComponentMask(None=VectorParameter() {group=Nitro, parameterName=BubbleSpeed_02, sortPriority=201, defaultValue=<rgba 0 0.35 0.2 1>}.RGB) {g=true, r=true}
  )
  NamedRerouteDeclaration(None=Add(A=Panner(Coordinate=@n61, Speed=@n62) {speedY=-0.4}, B=NamedRerouteUsage())) {name=Bubble_02_UV, nodeColor=<rgba 0 1 0.0391 1>}
  NamedRerouteDeclaration() {name=Debug, nodeColor=<rgba 1 0.2578 0 1>}
  NamedRerouteDeclaration(None=ChannelMaskParameter(None=FunctionInput() {bUsePreviewValueAsDefault=true, inputName=SkinFX, sortPriority=1, previewValue=<1 0 1 1>}) {group=05_Nitro, parameterName=NitroFX_Mask}.R) {name=Nitro_Mask, nodeColor=<rgba 0 0.5312 1 1>}
  n63 = NamedRerouteUsage()
  n64 = SetMaterialAttributes!(
   MaterialAttributes=NamedRerouteUsage()
   BaseColor=@n63
   EmissiveColor=@n63
   Normal=NamedRerouteUsage()
   Metallic=ScalarParameter() {group=Nitro, parameterName=Metallic, sortPriority=2}
   Specular=ScalarParameter() {defaultValue=1, group=Nitro, parameterName=Specular, sortPriority=3}
   Roughness=ScalarParameter() {defaultValue=0.1, group=Nitro, parameterName=Roughness, sortPriority=4}
  ) {attributeSetTypes [
    #1
    #2
    #3
    #4
    #5
    #6
   ]}
  NamedRerouteDeclaration(None=@n64) {name=MA_Out, nodeColor=<rgba 0 1 0.2171 1>}
  n65 = Time()
  n66 = ScalarParameter() {defaultValue=1, group=Nitro, parameterName=NitroPulseSpeed, sortPriority=30}
  n67 = Frac(None=Multiply(A=Multiply(A=@n65, B=@n66), B=Constant3Vector() {constant=<rgba 1.35 0.65 1 0>}))
  n68 = Frac(None=Multiply(A=Sine(None=@n67)))
  n69 = TextureCoordinate()
  n70 = ScalarParameter() {defaultValue=0.15, group=Nitro, parameterName=Volatile_NoiseMap_Tiling, sliderMax=1}
  n71 = TextureSample(UVs=Panner(
    Coordinate=Multiply(A=@n69, B=@n70)
    Speed=VectorParameter() {group=Nitro, parameterName=Volatile_NoiseMap_ScrollSpeed, defaultValue=<rgba 0.0333 0.1 0 0>}.RGB
   ) {speedY=0.1}) {samplerSource=SSM_Wrap_WorldGroupSettings, texture=&T_SoftNoise}
  n72 = DotProduct(A=Subtract(A=Multiply(A=@n68) {constB=2}), B=@n71.RGB)
  NamedRerouteDeclaration(None=LinearInterpolate(A=Saturate(None=@n72), Alpha=NamedRerouteUsage())) {name=Flash, nodeColor=<rgba 0 1 0.3906 1>}
  n73 = TextureSample(UVs=@n26, Tex=@MaterialExpressionTextureObject_1) {desc=heightTex}
  NamedRerouteDeclaration(None=MaterialFunctionCall("Xb1,PS4"=@n20.return, "PC-High&&Epic,XSX,PS5"=@n20.return, "PC-Perf,Android"=@n73.R, Switch=@n73.R, "PC-Low(NotPerf)&&Medium"=@n20.return) {materialFunction=&PlatformSwitch_1D}.Result) {name=RayMarchedHeightMap, nodeColor=<rgba 0 0.6719 1 1>}
  NamedRerouteDeclaration(None=ScalarParameter() {group=Nitro, parameterName=Volatile, sortPriority=80}) {name=Volatile, nodeColor=<rgba 1 0 0.1641 1>}
  NamedRerouteDeclaration(None=ScalarParameter() {defaultValue=1, group=Nitro, parameterName=GlobalSpeed, sortPriority=1}) {name=NitroGlobalSpeed, nodeColor=<rgba 0.9219 0 1 1>}
  NamedRerouteDeclaration(None=ComponentMask(None=VectorParameter() {group=Nitro, parameterName=GlobalTiling, sortPriority=0, defaultValue=<rgba 1 1 0 1>}.RGB) {g=true, r=true}) {name=NitroGlobalTiling, nodeColor=<rgba 0.8984 1 0 1>}
  n74 = TextureCoordinate()
  n75 = Multiply(
   A=ComponentMask(None=VectorParameter() {group=Nitro, parameterName=NitroDensityMaskTile, sortPriority=41, defaultValue=<rgba 0.5 0.5 0 1>}.RGB) {g=true, r=true}
   B=NamedRerouteUsage()
  )
  n76 = ComponentMask(None=VectorParameter() {group=Nitro, parameterName=NitroDensityMaskSpeed, sortPriority=42, defaultValue=<rgba 0 0.1 0 1>}.RGB) {g=true, r=true}
  n77 = NamedRerouteUsage()
  n78 = TextureSampleParameter2D(UVs=Panner(Coordinate=Multiply(A=@n74, B=@n75), Speed=Multiply(A=@n76, B=@n77))) {
   group=Nitro
   parameterName=NitroDensityMask
   samplerSource=SSM_Wrap_WorldGroupSettings
   sortPriority=40
   texture=/Game/Global/Textures/Default/Tileable_Patterns/T_Blowing_Smoke_Pattern2_Tile_DISP
  }
  NamedRerouteDeclaration(None=@n78.R) {name=DensityMask, nodeColor=<rgba 0 0.2969 1 1>}
  n79 = DotProduct(A=VertexNormalWS(), B=CameraVectorWS())
  n80 = ScalarParameter() {defaultValue=4, group=Nitro, parameterName=GlowFresnelMask_Exponnet, sliderMax=5}
  n81 = NamedRerouteUsage()
  n82 = NamedRerouteUsage()
  n83 = NamedRerouteUsage()
  n84 = Multiply(
   A=LinearInterpolate(B=Power(Base=@n79, Exp=@n80), Alpha=ScalarParameter() {defaultValue=1, group=Nitro, parameterName=GlowFresnelMask_Intensity, sliderMax=1}) {constA=1}
   B=StaticSwitch(True=LinearInterpolate(A=@n81, B=@n82, Alpha=@n83), False=@n81, Value=NamedRerouteUsage())
  )
  n85 = NamedRerouteUsage()
  n86 = ScalarParameter() {group=Nitro, parameterName=GlowOccludeDarken}
  n87 = OneMinus(None=Saturate(None=Multiply(A=@n85, B=@n86)))
  n88 = VectorParameter() {group=Nitro, parameterName=GlassEdgeColor, sortPriority=400, defaultValue=<rgba 0.2396 0.1355 0.0886 1>}
  n89 = Add(
   A=Multiply(A=NamedRerouteUsage(), B=ScalarParameter() {defaultValue=0.5, group=Nitro, parameterName=GlassEdgeColorStrength, sortPriority=410})
   B=Multiply(A=NamedRerouteUsage(), B=ScalarParameter() {group=Nitro, parameterName=GlassOccludeMaskStrength, sortPriority=411})
  )
  NamedRerouteDeclaration(None=LinearInterpolate(A=Multiply(A=@n84, B=@n87), B=@n88.RGB, Alpha=Saturate(None=@n89))) {name=OUT_Color, nodeColor=<rgba 0.6172 1 0 1>}
  ScalarParameter() {defaultValue=8, group=Nitro, parameterName=NitroRayMarchSteps, sliderMax=16, sliderMin=1, sortPriority=13}
  StaticSwitchParameter(True=@n56, False=@n63) {parameterName=Debug_BubbleGradient}
  TextureSample(UVs=@n24, Tex=@MaterialExpressionTextureObject_1) {desc=distortionTex}
 ]}
